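# MM_Blockout_Emissive1 is a Default Lit parent material for vehicle cosmetic blockouts, enabled for skeletal meshes and mesh deformers, apparently meant for solid emissive areas.
# Base Color is multiplied by Intensity, then by a fixed-alpha lerp between Intensity and a slightly boosted copy, so brightness scales roughly with Intensity squared.
# A time-driven sine using the Vibration Intenisty parameter sits separately, with no link to the color chain.
# No connections to material outputs are shown.
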Material /VehicleCosmetics/Materials/BlockOut_Materials/Vehicle_BO_Materials/MATID_SOLIDS/Parent/MM_Blockout_Emissive1
graph {
 properties {bAutomaticallySetUsageInEditor=false, bUsedWithMeshDeformer=true, bUsedWithSkeletalMesh=true, usedShadingModels=Default Lit}
 nodes [
  n1 = ScalarParameter() {defaultValue=5, parameterName=Intensity}
  Multiply(
   A=Multiply(A=VectorParameter() {parameterName=Base Color, defaultValue=<rgba 0.9527 0.9705 0.9412 1>}.RGB, B=@n1)
   B=LinearInterpolate(A=@n1, B=Divide(A=@n1) {constB=0.85})
  )
  Sine(None=Multiply(A=Time(), B=ScalarParameter() {defaultValue=15, parameterName=Vibration Intenisty}))
 ]
}
